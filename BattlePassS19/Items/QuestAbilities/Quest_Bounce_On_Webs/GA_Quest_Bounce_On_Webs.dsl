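; Event-triggered quest gameplay ability that credits a player for consecutive bounces (on webs, per its name) before landing.
; On activation it caches target tags, zeroes the counter and runs tasks for a landing event and bounce-tag additions.
; Each tag addition increments the count; reaching the maximum sends a quest stat event, stops both tasks, then ends once the bounce tag is removed.
; Landing earlier resets the count and ends the ability.
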(event Ability|EventActivateAbilityFromEvent
  (params (EventData :type Gameplay%20Event%20Data%20Structure%20%28by%20ref%29))
  (bind (_eventtag _instigator _target _optionalobject _optionalobject2 _contexthandle _instigatortags _targettags) (Utilities|Struct|BreakGameplayEventData EventData))
  (|CallParentFunction EventData)
  (Variables|Tags|SetTargetTags _targettags)
  (Development|PrintString "@@@ ActivateAbilityFromEvent [GA_Quest_Bounce_On_Webs]: CALLED - Tire Launch!" false (Variables|Debug|GetbDebugAbility))
  ;[1 0.2402 0.03963] Initialize bounce tracking...
  (Variables|Default|SetBounceCount 0)
  (bind _asynctaskproxy (Ability|Tasks|WaitGameplayEvent (Variables|Tags|GetTrackedLandingTag) 0 true)
    (:then
      (Variables|Default|SetEventTask _asynctaskproxy)
      (bind _asynctaskproxy_2 (Ability|Tasks|WaitGameplayTagAdd (Variables|Tags|GetTrackedBounceTag))
        (:then
          (Variables|Default|SetTagTask _asynctaskproxy_2))
        (:Added
          (bind _result (Math|Integer|IncrementInt (Variables|Default|GetBounceCount))
            (:%20%20
              (bind _output_get (Variables|Default|SetBounceCount _result))
              (if (>= _output_get (Variables|Default|GetMaxBounces))
                ;[0.03963 0.9403 1] Max bounce count reached!
                (Development|PrintString "@@@@@@@@@ ActivateAbilityFromEvent [GA_Quest_Bounce_On_Webs]: Target bounce count reached!" false (Variables|Debug|GetbDebugAbility))
                (Quest|SendComplexCustomStatEvent (Variables|GAVars|Data|GetOwningPlayerQuestManager) 0 (Variables|Tags|GetStatEventSourceTags) (Variables|Tags|GetTargetTags) 1)
                (GameplayTasks|EndTask (Variables|Default|GetTagTask))
                (GameplayTasks|EndTask (Variables|Default|GetEventTask))
                (Ability|Tasks|WaitGameplayTagRemove (Variables|Tags|GetTrackedBounceTag) 0 true
                  (:Removed
                    (Ability|EndAbility)))
                (else
                  (Development|PrintString (Utilities|String|Append "@@@@@@ ActivateAbilityFromEvent [GA_Quest_Bounce_On_Webs]: Target bounced count = " (Utilities|String|ToString(Integer) (Variables|Default|GetBounceCount)) "...") false (Variables|Debug|GetbDebugAbility)))))))))
    (:EventReceived
      ;[1 0.09399 0.7665] Target landed...
      (Variables|Default|SetBounceCount 0)
      (Development|PrintString "@@@@@@ ActivateAbilityFromEvent [GA_Quest_Bounce_On_Webs]: Landed - resetting counter and ending..." false (Variables|Debug|GetbDebugAbility))
      (GameplayTasks|EndTask (Variables|Default|GetTagTask))
      (Ability|EndAbility))))
